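# Post-copy Anim Blueprint for the female medium Ashton Boardwalk cosmetic body, adding secondary motion to its accessory bones.
# It copies the pose from the attached parent mesh, then simulates the right-arm bandana bone with Anim Dynamics using an angular spring, limits and reduced gravity.
# Two Constraint nodes then drive the left and right gun bones from weighted blends of pelvis, thigh and thigh-twist bones.
# The event graph has no logic.
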AnimBlueprint /BRCosmetics/Characters/Player/Female/Medium/Bodies/F_MED_AshtonBoardwalk/Meshes/F_MED_AshtonBoardwalk_AnimBP
target_skeleton=&/F_MED_AshtonBoardwalk_Skeleton
graphs [
 AnimGraph {nodes [
   n1 = Anim Dynamics - Bone: dyn_arm_r_bandana(ComponentPose=Local To Component(LocalPose=Copy Pose From Mesh() {bUseAttachedParent=true}), Alpha=0.75, ExternalForce="0,0,0") {
    alpha=0.9
    angularDampingOverride=0.985
    angularSpringConstant=300
    bAngularSpring=true
    bGravityOverrideInSimSpace=true
    bOverrideAngularDamping=true
    gravityScale=0.3
    numSolverIterationsPostUpdate=4
    numSolverIterationsPreUpdate=16
    alphaBoolBlend {blendOption=Linear}
    boundBone {boneName=dyn_arm_r_bandana}
    physicsBodyDefinitions [{
      sphereCollisionRadius=0
      boundBone {boneName=dyn_arm_r_bandana}
      boxExtents=<20 8 8>
      constraintSetup {angularTargetAxis=AxisZ, twistAxis=AxisY, angularLimitsMax=<15 0 15>, angularLimitsMin=<-15 0 -15>, angularTarget=<0 0 1>}
      localJointOffset=<0 0 -3>
     }]
   }
   n2 = "Constraint - dyn_gun_r (PPP)"!(ComponentPose=@n1, Alpha=1, ConstraintWeights_0=0.6546, ConstraintWeights_1:float=0.0357, ConstraintWeights_2:float=0.3097) -> (Pose) {
    alphaBoolBlend {blendOption=Linear}
    boneToModify {boneName=dyn_gun_r}
    constraintSetup [{targetBone {boneName=thigh_r}}, {targetBone {boneName=pelvis}}, {targetBone {boneName=thigh_twist_01_r}}]
    constraintWeights [1,1,1]
   }
   n3 = "Constraint - dyn_gun_l (PPP)"!(ComponentPose=@n2, Alpha=1, ConstraintWeights_0=0.0349, ConstraintWeights_1:float=0.3091, ConstraintWeights_2:float=0.656) -> (Pose) {
    alphaBoolBlend {blendOption=Linear}
    boneToModify {boneName=dyn_gun_l}
    constraintSetup [{targetBone {boneName=pelvis}}, {targetBone {boneName=thigh_twist_01_l}}, {targetBone {boneName=thigh_l}}]
    constraintWeights [1,1,1]
   }
   Output Pose(Result=Component To Local(ComponentPose=@n3)) {layerGroup=None}
  ]}
 EventGraph {nodes [Event BlueprintUpdateAnimation()]}
]
variables []
