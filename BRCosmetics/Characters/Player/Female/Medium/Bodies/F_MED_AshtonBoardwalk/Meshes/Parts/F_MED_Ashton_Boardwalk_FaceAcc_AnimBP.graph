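# Post-process-style AnimBP that animates the hair of the female Ashton Boardwalk cosmetic's FaceAcc mesh.
# It copies the parent mesh pose, then runs chained AnimDynamics springs on hair bones, with spherical and planar limits around neck, spine and clavicles.
# Additive rotations adjust strands for left-hair rotation and raised right arm before caching; crouch-moving blends in further right/back offsets.
# Several other transform chains are unconnected, and the EventGraph is empty.
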AnimBlueprint /BRCosmetics/Characters/Player/Female/Medium/Bodies/F_MED_AshtonBoardwalk/Meshes/Parts/F_MED_Ashton_Boardwalk_FaceAcc_AnimBP
target_skeleton=&/F_MED_Ashton_Boardwalk_FaceAcc_Skeleton
graphs [
 AnimGraph {nodes [
   "Transform (Modify) Bone - Bone: dyn_hair_r_02"(Alpha=1, Translation="0,0,0", Rotation=Get HairRightRotation() -> (HairRightRotation:Rotator), Scale="1,1,1") {rotationMode=BMM_Additive, rotationSpace=BCS_BoneSpace, alphaBoolBlend {blendOption=Linear}, boneToModify {boneName=dyn_hair_r_02}}
   n1 = Get ArmUpRightAlpha() -> (ArmUpRightAlpha:float)
   n2 = "Transform (Modify) Bone - Bone: dyn_hair_r_01"(Alpha=@n1, Translation="0,0,0", Rotation="-5.023,37.891,-13.100", Scale="1,1,1") {
    rotationMode=BMM_Additive
    rotationSpace=BCS_BoneSpace
    alphaBoolBlend {blendOption=Linear}
    boneToModify {boneName=dyn_hair_r_01}
    rotation=<rot -5.023 37.891 -13.1>
   }
   n3 = "Transform (Modify) Bone - Bone: dyn_hair_r_02"(ComponentPose=@n2, Alpha=@n1, Translation="0,0,0", Rotation="13.514,26.441,-12.521", Scale="1,1,1") {
    rotationMode=BMM_Additive
    rotationSpace=BCS_BoneSpace
    alphaBoolBlend {blendOption=Linear}
    boneToModify {boneName=dyn_hair_r_02}
    rotation=<rot 13.514 26.4408 -12.5211>
   }
   "Transform (Modify) Bone - Bone: dyn_hair_r_03"(ComponentPose=@n3, Alpha=@n1, Translation="0,0,0", Rotation="0.000,0.000,77.988", Scale="1,1,1") {
    rotationMode=BMM_Additive
    rotationSpace=BCS_BoneSpace
    alphaBoolBlend {blendOption=Linear}
    boneToModify {boneName=dyn_hair_r_03}
    rotation {roll=77.9879}
   }
   n4 = Get CrouchAlpha() -> (CrouchAlpha:float)
   n5 = "Transform (Modify) Bone - Bone: dyn_hair_r_01"(Alpha=@n4, Translation="0,0,0", Rotation="-24.680,33.628,-15.522", Scale="1,1,1") {
    rotationMode=BMM_Additive
    rotationSpace=BCS_BoneSpace
    alphaBoolBlend {blendOption=Linear}
    alphaScaleBiasClamp {bInterpResult=true, interpSpeedDecreasing=15, interpSpeedIncreasing=15}
    boneToModify {boneName=dyn_hair_r_01}
    rotation=<rot -24.68 33.628 -15.522>
   }
   n6 = "Transform (Modify) Bone - Bone: dyn_hair_r_02"(ComponentPose=@n5, Alpha=@n4, Translation="0,0,0", Rotation="-11.055,-5.905,1.135", Scale="1,1,1") {
    rotationMode=BMM_Additive
    rotationSpace=BCS_BoneSpace
    alphaBoolBlend {blendOption=Linear}
    alphaScaleBiasClamp {bInterpResult=true, interpSpeedDecreasing=15, interpSpeedIncreasing=15}
    boneToModify {boneName=dyn_hair_r_02}
    rotation=<rot -11.055 -5.9051 1.135>
   }
   n7 = "Transform (Modify) Bone - Bone: dyn_hair_bk_r_01"(ComponentPose=@n6, Alpha=@n4, Translation="0,0,0", Rotation="11.481,-34.535,0.384", Scale="1,1,1") {
    rotationMode=BMM_Additive
    rotationSpace=BCS_BoneSpace
    alphaBoolBlend {blendOption=Linear}
    alphaScaleBiasClamp {bInterpResult=true, interpSpeedDecreasing=15, interpSpeedIncreasing=15}
    boneToModify {boneName=dyn_hair_bk_r_01}
    rotation=<rot 11.481 -34.5351 0.384>
   }
   "Transform (Modify) Bone - Bone: dyn_hair_bk_r_02"(ComponentPose=@n7, Alpha=@n4, Translation="0,0,0", Rotation="-8.549,-25.465,4.051", Scale="1,1,1") {
    rotationMode=BMM_Additive
    rotationSpace=BCS_BoneSpace
    alphaBoolBlend {blendOption=Linear}
    alphaScaleBiasClamp {bInterpResult=true, interpSpeedDecreasing=15, interpSpeedIncreasing=15}
    boneToModify {boneName=dyn_hair_bk_r_02}
    rotation=<rot -8.549 -25.4649 4.051>
   }
   n8 = "Transform (Modify) Bone - Bone: dyn_hair_bk_l_01"(Alpha=1, Translation="0,0,0", Rotation="1.918,-46.036,0.521", Scale="1,1,1") {
    rotationMode=BMM_Additive
    rotationSpace=BCS_BoneSpace
    alphaBoolBlend {blendOption=Linear}
    alphaScaleBiasClamp {bInterpResult=true, interpSpeedDecreasing=15, interpSpeedIncreasing=15}
    boneToModify {boneName=dyn_hair_bk_l_01}
    rotation=<rot 1.918 -46.0361 0.521>
   }
   n9 = "Transform (Modify) Bone - Bone: dyn_hair_bk_l_02"(ComponentPose=@n8, Alpha=1, Translation="0,0,0", Rotation="-3.616,-21.636,0.183", Scale="1,1,1") {
    rotationMode=BMM_Additive
    rotationSpace=BCS_BoneSpace
    alphaBoolBlend {blendOption=Linear}
    alphaScaleBiasClamp {bInterpResult=true, interpSpeedDecreasing=15, interpSpeedIncreasing=15}
    boneToModify {boneName=dyn_hair_bk_l_02}
    rotation=<rot -3.616 -21.636 0.183>
   }
   n10 = "Transform (Modify) Bone - Bone: dyn_hair_l_01"(ComponentPose=@n9, Alpha=1, Translation="0,0,0", Rotation="27.484,19.159,4.791", Scale="1,1,1") {
    rotationMode=BMM_Additive
    rotationSpace=BCS_BoneSpace
    alphaBoolBlend {blendOption=Linear}
    alphaScaleBiasClamp {bInterpResult=true, interpSpeedDecreasing=15, interpSpeedIncreasing=15}
    boneToModify {boneName=dyn_hair_l_01}
    rotation=<rot 27.484 19.159 4.791>
   }
   "Transform (Modify) Bone - Bone: dyn_hair_l_02"(ComponentPose=@n10, Alpha=1, Translation="0,0,0", Rotation="5.177,4.598,0.601", Scale="1,1,1") {
    rotationMode=BMM_Additive
    rotationSpace=BCS_BoneSpace
    alphaBoolBlend {blendOption=Linear}
    alphaScaleBiasClamp {bInterpResult=true, interpSpeedDecreasing=15, interpSpeedIncreasing=15}
    boneToModify {boneName=dyn_hair_l_02}
    rotation=<rot 5.177 4.598 0.601>
   }
   n11 = "Transform (Modify) Bone - Bone: dyn_hair_r_01"(
    ComponentPose=Local To Component(LocalPose=Use cached pose 'PreCrouchCache'())
    Alpha=0.7
    Translation="0,0,0"
    Rotation="-30.837,21.175,-9.310"
    Scale="1,1,1"
   ) {
    rotationMode=BMM_Additive
    rotationSpace=BCS_BoneSpace
    alphaBoolBlend {blendOption=Linear}
    alphaScaleBiasClamp {bInterpResult=true, interpSpeedDecreasing=15, interpSpeedIncreasing=15}
    boneToModify {boneName=dyn_hair_r_01}
    rotation=<rot -30.837 21.175 -9.31>
   }
   n12 = "Transform (Modify) Bone - Bone: dyn_hair_r_02"(ComponentPose=@n11, Alpha=0.7, Translation="0,0,0", Rotation="-11.056,-5.859,1.127", Scale="1,1,1") {
    rotationMode=BMM_Additive
    rotationSpace=BCS_BoneSpace
    alphaBoolBlend {blendOption=Linear}
    alphaScaleBiasClamp {bInterpResult=true, interpSpeedDecreasing=15, interpSpeedIncreasing=15}
    boneToModify {boneName=dyn_hair_r_02}
    rotation=<rot -11.056 -5.859 1.127>
   }
   n13 = "Transform (Modify) Bone - Bone: dyn_hair_bk_r_01"(ComponentPose=@n12, Alpha=0.7, Translation="0,0,0", Rotation="11.418,-30.059,1.272", Scale="1,1,1") {
    rotationMode=BMM_Additive
    rotationSpace=BCS_BoneSpace
    alphaBoolBlend {blendOption=Linear}
    alphaScaleBiasClamp {bInterpResult=true, interpSpeedDecreasing=15, interpSpeedIncreasing=15}
    boneToModify {boneName=dyn_hair_bk_r_01}
    rotation=<rot 11.418 -30.059 1.272>
   }
   n14 = "Transform (Modify) Bone - Bone: dyn_hair_bk_r_02"(ComponentPose=@n13, Alpha=0.7, Translation="0,0,0", Rotation="-3.148,10.885,-1.788", Scale="1,1,1") {
    rotationMode=BMM_Additive
    rotationSpace=BCS_BoneSpace
    alphaBoolBlend {blendOption=Linear}
    alphaScaleBiasClamp {bInterpResult=true, interpSpeedDecreasing=15, interpSpeedIncreasing=15}
    boneToModify {boneName=dyn_hair_bk_r_02}
    rotation=<rot -3.148 10.8851 -1.788>
   }
   n15 = "Transform (Modify) Bone - Bone: dyn_hair_bk_01"(ComponentPose=@n14, Alpha=0.7, Translation="0,0,0", Rotation="11.418,-30.059,1.272", Scale="1,1,1") {
    rotationMode=BMM_Additive
    rotationSpace=BCS_BoneSpace
    alphaBoolBlend {blendOption=Linear}
    alphaScaleBiasClamp {bInterpResult=true, interpSpeedDecreasing=15, interpSpeedIncreasing=15}
    boneToModify {boneName=dyn_hair_bk_01}
    rotation=<rot 11.418 -30.059 1.272>
   }
   n16 = "Transform (Modify) Bone - Bone: dyn_hair_bk_02"(ComponentPose=@n15, Alpha=0.7, Translation="0,0,0", Rotation="-3.148,10.885,-1.788", Scale="1,1,1") {
    rotationMode=BMM_Additive
    rotationSpace=BCS_BoneSpace
    alphaBoolBlend {blendOption=Linear}
    alphaScaleBiasClamp {bInterpResult=true, interpSpeedDecreasing=15, interpSpeedIncreasing=15}
    boneToModify {boneName=dyn_hair_bk_02}
    rotation=<rot -3.148 10.8851 -1.788>
   }
   Output Pose(Result=Blend Poses by bool(
     BlendPose_0=Component To Local(ComponentPose=@n16)
     BlendPose_1=Use cached pose 'PreCrouchCache'()
     bActiveValue=Get bIsCrouchMoving() -> (bIsCrouchMoving:bool)
    ) {blendType=Linear}) {layerGroup=None}
   n17 = Get HairAlpha() -> (HairAlpha:double)
   n18 = Anim Dynamics - Bone: dyn_hair_bk_01(ComponentPose=Local To Component(LocalPose=Copy Pose From Mesh() {bUseAttachedParent=true}), Alpha=@n17, ExternalForce="0,0,0") {
    angularDampingOverride=0.9999
    angularSpringConstant=1000
    bAngularSpring=true
    bGravityOverrideInSimSpace=true
    bOverrideAngularDamping=true
    bUsePlanarLimit=false
    bUseSphericalLimits=true
    gravityScale=0.33
    alphaBoolBlend {blendOption=Linear}
    boundBone {boneName=dyn_hair_bk_01}
    componentAppliedLinearAccClamp=<100 100 100>
    componentLinearVelScale=<50 50 1>
    physicsBodyDefinitions [{
      collisionType=CustomSphere
      sphereCollisionRadius=3
      boundBone {boneName=dyn_hair_bk_01}
      boxExtents=<5 30 30>
      constraintSetup {angularLimitsMax=<0 20 35>, angularLimitsMin=<0 -20 -65>, angularTarget=<0 0 0>}
      localJointOffset=<6.5 0 0>
     }]
    sphericalLimits [{limitRadius=6, drivingBone {boneName=neck_02}, sphereLocalOffset=<0 0 0>}]
   }
   n19 = Anim Dynamics - Bone: dyn_hair_bk_02(ComponentPose=@n18, Alpha=@n17, ExternalForce="0,0,0") {
    angularDampingOverride=0.9999
    angularSpringConstant=1000
    bAngularSpring=true
    bGravityOverrideInSimSpace=true
    bOverrideAngularDamping=true
    bUseSphericalLimits=true
    gravityScale=0.35
    alphaBoolBlend {blendOption=Linear}
    boundBone {boneName=dyn_hair_bk_02}
    componentAppliedLinearAccClamp=<100 100 100>
    componentLinearVelScale=<50 50 1>
    physicsBodyDefinitions [{
      collisionType=CustomSphere
      sphereCollisionRadius=5
      boundBone {boneName=dyn_hair_bk_02}
      boxExtents=<5 40 40>
      constraintSetup {angularLimitsMax=<0 20 35>, angularLimitsMin=<0 -20 -75>, angularTarget=<0 0 0>}
      localJointOffset=<10 0 0>
     }]
    sphericalLimits [{limitRadius=41, drivingBone {boneName=spine_05}, sphereLocalOffset=<-12 26 0>}]
   }
   n20 = Anim Dynamics - Bone: dyn_hair_bk_r_01(ComponentPose=@n19, Alpha=@n17, ExternalForce="0,0,0") {
    angularDampingOverride=1
    angularSpringConstant=1000
    bAngularSpring=true
    bGravityOverrideInSimSpace=true
    bOverrideAngularDamping=true
    bUsePlanarLimit=false
    gravityScale=0.45
    alphaBoolBlend {blendOption=Linear}
    boundBone {boneName=dyn_hair_bk_r_01}
    componentAppliedLinearAccClamp=<100 100 100>
    componentLinearVelScale=<50 50 50>
    physicsBodyDefinitions [{
      collisionType=CustomSphere
      sphereCollisionRadius=3
      boundBone {boneName=dyn_hair_bk_r_01}
      boxExtents=<5 30 30>
      constraintSetup {angularLimitsMax=<15 20 30>, angularLimitsMin=<-15 -35 -65>, angularTarget=<0 0 0>}
      localJointOffset=<6.5 0 -1>
     }]
    sphericalLimits [
     {limitRadius=5, drivingBone {boneName=neck_02}, sphereLocalOffset=<0 0 0>}
     {limitRadius=45, drivingBone {boneName=spine_05}, sphereLocalOffset=<-30 20 8>}
    ]
   }
   n21 = Anim Dynamics - Bone: dyn_hair_bk_r_02(ComponentPose=@n20, Alpha=@n17, ExternalForce="0,0,0") {
    angularDampingOverride=1
    angularSpringConstant=1000
    bAngularSpring=true
    bGravityOverrideInSimSpace=true
    bOverrideAngularDamping=true
    bUseSphericalLimits=true
    gravityScale=0.6
    alphaBoolBlend {blendOption=Linear}
    boundBone {boneName=dyn_hair_bk_r_02}
    componentAppliedLinearAccClamp=<100 100 100>
    componentLinearVelScale=<50 50 50>
    physicsBodyDefinitions [{
      collisionType=CustomSphere
      sphereCollisionRadius=2.2
      boundBone {boneName=dyn_hair_bk_r_02}
      boxExtents=<5 40 40>
      constraintSetup {angularLimitsMax=<0 20 30>, angularLimitsMin=<0 -20 -75>, angularTarget=<0 0 0>}
      localJointOffset=<4 0 0>
     }]
    sphericalLimits [
     {limitRadius=45, drivingBone {boneName=spine_05}, sphereLocalOffset=<-30 8 8>}
     {limitRadius=44, drivingBone {boneName=clavicle_r}, sphereLocalOffset=<0 -10 34>}
     {limitRadius=5, drivingBone {boneName=neck_02}, sphereLocalOffset=<0 0 0>}
    ]
   }
   n22 = Anim Dynamics - Bone: dyn_hair_bk_r_03(ComponentPose=@n21, Alpha=1, ExternalForce="0,0,0") {
    angularDampingOverride=1
    angularSpringConstant=1000
    bAngularSpring=true
    bGravityOverrideInSimSpace=true
    bOverrideAngularDamping=true
    bUseSphericalLimits=true
    gravityScale=0.6
    alphaBoolBlend {blendOption=Linear}
    boundBone {boneName=dyn_hair_bk_r_03}
    componentAppliedLinearAccClamp=<100 100 100>
    componentLinearVelScale=<50 50 50>
    physicsBodyDefinitions [{
      collisionType=CustomSphere
      sphereCollisionRadius=2.2
      boundBone {boneName=dyn_hair_bk_r_03}
      boxExtents=<5 40 40>
      constraintSetup {angularLimitsMax=<0 20 30>, angularLimitsMin=<0 -20 -75>, angularTarget=<0 0 0>}
      localJointOffset=<1 0 0>
     }]
    sphericalLimits [
     {limitRadius=45, drivingBone {boneName=spine_05}, sphereLocalOffset=<-30 0 8>}
     {limitRadius=44, drivingBone {boneName=clavicle_r}, sphereLocalOffset=<0 0 34>}
     {limitRadius=6.5, drivingBone {boneName=neck_02}, sphereLocalOffset=<0 0 0>}
    ]
   }
   n23 = Anim Dynamics - Bone: dyn_hair_r_01(ComponentPose=@n22, Alpha=@n17, ExternalForce="0,0,0") {
    angularDampingOverride=1
    angularSpringConstant=1000
    bAngularSpring=true
    bGravityOverrideInSimSpace=true
    bOverrideAngularDamping=true
    bUsePlanarLimit=false
    bUseSphericalLimits=true
    gravityScale=0.45
    alphaBoolBlend {blendOption=Linear}
    boundBone {boneName=dyn_hair_r_01}
    componentAppliedLinearAccClamp=<100 100 100>
    componentLinearVelScale=<50 50 50>
    physicsBodyDefinitions [{
      collisionType=CustomSphere
      sphereCollisionRadius=3
      boundBone {boneName=dyn_hair_r_01}
      boxExtents=<5 30 30>
      constraintSetup {angularLimitsMax=<10 25 17>, angularLimitsMin=<-10 -35 -25>, angularTarget=<0 0 0>}
      localJointOffset=<6 0 0>
     }]
    sphericalLimits [{limitRadius=6, drivingBone {boneName=neck_02}, sphereLocalOffset=<0 0 0>}]
   }
   n24 = Anim Dynamics - Bone: dyn_hair_r_02(ComponentPose=@n23, Alpha=1, ExternalForce="0,0,0") {
    angularDampingOverride=1
    angularSpringConstant=1000
    bAngularSpring=true
    bGravityOverrideInSimSpace=true
    bOverrideAngularDamping=true
    bUsePlanarLimit=false
    bUseSphericalLimits=true
    gravityScale=0.45
    alphaBoolBlend {blendOption=Linear}
    boundBone {boneName=dyn_hair_r_02}
    componentAppliedLinearAccClamp=<100 100 100>
    componentLinearVelScale=<50 50 50>
    physicsBodyDefinitions [{
      collisionType=CustomSphere
      sphereCollisionRadius=3
      boundBone {boneName=dyn_hair_r_02}
      boxExtents=<8 30 30>
      constraintSetup {angularLimitsMax=<10 45 17>, angularLimitsMin=<-10 -35 -25>, angularTarget=<0 0 0>}
      localJointOffset=<3 0 0>
     }]
    sphericalLimits [
     {limitRadius=5, drivingBone {boneName=neck_02}, sphereLocalOffset=<0 0 0>}
     {limitRadius=42, drivingBone {boneName=spine_05}, sphereLocalOffset=<-29 5 0>}
    ]
   }
   n25 = Anim Dynamics - Bone: dyn_hair_r_03(ComponentPose=@n24, Alpha=@n17, ExternalForce="0,0,0") {
    angularDampingOverride=1
    angularSpringConstant=1000
    bAngularSpring=true
    bGravityOverrideInSimSpace=true
    bOverrideAngularDamping=true
    bUsePlanarLimit=false
    bUseSphericalLimits=true
    gravityScale=0.45
    alphaBoolBlend {blendOption=Linear}
    boundBone {boneName=dyn_hair_r_03}
    componentAppliedLinearAccClamp=<100 100 100>
    componentLinearVelScale=<50 50 50>
    physicsBodyDefinitions [{
      collisionType=CustomSphere
      sphereCollisionRadius=3
      boundBone {boneName=dyn_hair_r_03}
      boxExtents=<8 30 30>
      constraintSetup {angularLimitsMax=<10 45 5>, angularLimitsMin=<-10 -15 -30>, angularTarget=<0 0 0>}
      localJointOffset=<1 0 0>
     }]
    sphericalLimits [
     {limitRadius=42, drivingBone {boneName=spine_05}, sphereLocalOffset=<-29 0 0>}
     {limitRadius=33, drivingBone {boneName=clavicle_r}, sphereLocalOffset=<-10 0 25>}
     {limitRadius=6.5, drivingBone {boneName=neck_02}, sphereLocalOffset=<0 0 0>}
    ]
   }
   n26 = Anim Dynamics - Bone: dyn_hair_bk_l_01(ComponentPose=@n25, Alpha=@n17, ExternalForce="0,0,0") {
    angularDampingOverride=1
    angularSpringConstant=1000
    bAngularSpring=true
    bGravityOverrideInSimSpace=true
    bOverrideAngularDamping=true
    bUsePlanarLimit=false
    gravityScale=0.45
    alphaBoolBlend {blendOption=Linear}
    boundBone {boneName=dyn_hair_bk_l_01}
    componentAppliedLinearAccClamp=<100 100 100>
    componentLinearVelScale=<50 50 50>
    physicsBodyDefinitions [{
      collisionType=CustomSphere
      sphereCollisionRadius=3
      boundBone {boneName=dyn_hair_bk_l_01}
      boxExtents=<5 30 30>
      constraintSetup {angularLimitsMax=<0 20 30>, angularLimitsMin=<0 -20 -65>, angularTarget=<0 0 0>}
      localJointOffset=<6.5 0 0>
     }]
    sphericalLimits [{limitRadius=5, drivingBone {boneName=neck_02}, sphereLocalOffset=<0 0 0>}]
   }
   n27 = Anim Dynamics - Bone: dyn_hair_bk_l_02(ComponentPose=@n26, Alpha=@n17, ExternalForce="0,0,0") {
    angularDampingOverride=1
    angularSpringConstant=1000
    bAngularSpring=true
    bGravityOverrideInSimSpace=true
    bOverrideAngularDamping=true
    gravityScale=0.7
    alphaBoolBlend {blendOption=Linear}
    boundBone {boneName=dyn_hair_bk_l_02}
    componentAppliedLinearAccClamp=<100 100 100>
    componentLinearVelScale=<50 50 50>
    physicsBodyDefinitions [{
      collisionType=CustomSphere
      sphereCollisionRadius=2.5
      boundBone {boneName=dyn_hair_bk_l_02}
      boxExtents=<5 40 40>
      constraintSetup {angularLimitsMax=<0 20 30>, angularLimitsMin=<0 -20 -75>, angularTarget=<0 0 0>}
      localJointOffset=<10 0 0>
     }]
    sphericalLimits [{limitRadius=43.5, drivingBone {boneName=spine_05}, sphereLocalOffset=<-31 22 -8>}]
   }
   n28 = Anim Dynamics - Bone: dyn_hair_l_01(ComponentPose=@n27, Alpha=@n17, ExternalForce="0,0,0") {
    angularDampingOverride=1
    angularSpringConstant=1000
    bAngularSpring=true
    bGravityOverrideInSimSpace=true
    bOverrideAngularDamping=true
    bUsePlanarLimit=false
    gravityScale=0.45
    alphaBoolBlend {blendOption=Linear}
    boundBone {boneName=dyn_hair_l_01}
    componentAppliedLinearAccClamp=<100 100 100>
    componentLinearVelScale=<50 50 50>
    physicsBodyDefinitions [{
      collisionType=CustomSphere
      sphereCollisionRadius=2
      boundBone {boneName=dyn_hair_l_01}
      boxExtents=<5 30 30>
      constraintSetup {angularLimitsMax=<0 8 10>, angularLimitsMin=<0 -8 -25>, angularTarget=<0 0 0>}
      localJointOffset=<-6.5 0 0>
     }]
    sphericalLimits [
     {limitRadius=5, drivingBone {boneName=neck_02}, sphereLocalOffset=<0 0 0>}
     {limitRadius=7, drivingBone {boneName=neck_01}, sphereLocalOffset=<5 10 0>}
    ]
   }
   n29 = Anim Dynamics - Bone: dyn_hair_l_02(ComponentPose=@n28, Alpha=@n17, ExternalForce="0,0,0") {
    angularDampingOverride=1
    angularSpringConstant=1000
    bAngularSpring=true
    bGravityOverrideInSimSpace=true
    bOverrideAngularDamping=true
    gravityScale=0.45
    alphaBoolBlend {blendOption=Linear}
    boundBone {boneName=dyn_hair_l_02}
    componentAppliedLinearAccClamp=<100 100 100>
    componentLinearVelScale=<50 50 50>
    physicsBodyDefinitions [{
      collisionType=CustomSphere
      sphereCollisionRadius=2
      boundBone {boneName=dyn_hair_l_02}
      boxExtents=<5 30 30>
      constraintSetup {angularLimitsMax=<0 8 40>, angularLimitsMin=<0 -8 -15>, angularTarget=<0 0 0>}
      localJointOffset=<-6.5 0 0>
     }]
    planarLimits [{drivingBone {boneName=clavicle_l}, planeTransform {location=<0 -2 0>, rotation=<rot -90 90 0>}}]
    sphericalLimits [
     {limitRadius=5, drivingBone {boneName=neck_02}, sphereLocalOffset=<0 0 0>}
     {limitRadius=7, drivingBone {boneName=neck_01}, sphereLocalOffset=<5 10 0>}
    ]
   }
   n30 = Anim Dynamics - Bone: dyn_hair_l_03(ComponentPose=@n29, Alpha=@n17, ExternalForce="0,0,0") {
    angularDampingOverride=1
    angularSpringConstant=1000
    bAngularSpring=true
    bGravityOverrideInSimSpace=true
    bOverrideAngularDamping=true
    bUsePlanarLimit=false
    gravityScale=0.45
    alphaBoolBlend {blendOption=Linear}
    boundBone {boneName=dyn_hair_l_03}
    componentAppliedLinearAccClamp=<100 100 100>
    componentLinearVelScale=<50 50 50>
    physicsBodyDefinitions [{
      collisionType=CustomSphere
      sphereCollisionRadius=2
      boundBone {boneName=dyn_hair_l_03}
      boxExtents=<5 30 30>
      constraintSetup {angularLimitsMax=<0 10 45>, angularLimitsMin=<0 -10 -25>, angularTarget=<0 0 0>}
      localJointOffset=<-6.5 0 0>
     }]
    sphericalLimits [
     {limitRadius=5, drivingBone {boneName=neck_02}, sphereLocalOffset=<0 0 0>}
     {limitRadius=7, drivingBone {boneName=neck_01}, sphereLocalOffset=<5 10 0>}
    ]
   }
   n31 = Anim Dynamics - Bone: dyn_hair_fr_r_01(ComponentPose=@n30, Alpha=@n17, ExternalForce="0,0,0") {
    angularDampingOverride=0.99
    angularSpringConstant=1000
    bAngularSpring=true
    bGravityOverrideInSimSpace=true
    bOverrideAngularDamping=true
    bUsePlanarLimit=false
    gravityScale=0.25
    alphaBoolBlend {blendOption=Linear}
    boundBone {boneName=dyn_hair_fr_r_01}
    componentAppliedLinearAccClamp=<100 100 100>
    componentLinearVelScale=<50 50 50>
    physicsBodyDefinitions [{
      collisionType=CustomSphere
      sphereCollisionRadius=3
      boundBone {boneName=dyn_hair_fr_r_01}
      boxExtents=<5 30 30>
      constraintSetup {angularLimitsMax=<0 25 20>, angularLimitsMin=<0 -2 -30>, angularTarget=<0 0 0>}
      localJointOffset=<-6.5 0 0>
     }]
    sphericalLimits [
     {limitRadius=5, drivingBone {boneName=neck_02}, sphereLocalOffset=<0 0 0>}
     {limitRadius=7, drivingBone {boneName=neck_01}, sphereLocalOffset=<5 10 0>}
    ]
   }
   n32 = Anim Dynamics - Bone: dyn_hair_fr_r_02(ComponentPose=@n31, Alpha=@n17, ExternalForce="0,0,0") {
    angularDampingOverride=0.99
    angularSpringConstant=1000
    bAngularSpring=true
    bGravityOverrideInSimSpace=true
    bOverrideAngularDamping=true
    bUsePlanarLimit=false
    gravityScale=0.3
    alphaBoolBlend {blendOption=Linear}
    boundBone {boneName=dyn_hair_fr_r_02}
    componentAppliedLinearAccClamp=<100 100 100>
    componentLinearVelScale=<50 50 50>
    physicsBodyDefinitions [{
      collisionType=CustomSphere
      sphereCollisionRadius=3
      boundBone {boneName=dyn_hair_fr_r_02}
      boxExtents=<5 30 30>
      constraintSetup {angularLimitsMax=<0 20 30>, angularLimitsMin=<0 -5 -40>, angularTarget=<0 0 0>}
      localJointOffset=<-8 0 0>
     }]
    sphericalLimits [{limitRadius=30, drivingBone {boneName=spine_05}, sphereLocalOffset=<10 7 -24>}]
   }
   n33 = "Transform (Modify) Bone - Bone: dyn_hair_fl_01"(ComponentPose=@n32, Alpha=1, Translation="0,0,0", Rotation="10,0,0", Scale="1,1,1") {
    rotationMode=BMM_Additive
    rotationSpace=BCS_BoneSpace
    alphaBoolBlend {blendOption=Linear}
    boneToModify {boneName=dyn_hair_fl_01}
    rotation {pitch=10}
   }
   n34 = Anim Dynamics - Bone: dyn_hair_fl_01(ComponentPose=@n33, Alpha=@n17, ExternalForce="0,0,0") {
    angularDampingOverride=0.99
    angularSpringConstant=1000
    bAngularSpring=true
    bGravityOverrideInSimSpace=true
    bOverrideAngularDamping=true
    bUsePlanarLimit=false
    gravityScale=0.25
    alphaBoolBlend {blendOption=Linear}
    boundBone {boneName=dyn_hair_fl_01}
    componentAppliedLinearAccClamp=<100 100 100>
    componentLinearVelScale=<50 50 50>
    physicsBodyDefinitions [{
      collisionType=CustomSphere
      sphereCollisionRadius=3
      boundBone {boneName=dyn_hair_fl_01}
      boxExtents=<5 30 30>
      constraintSetup {angularLimitsMax=<1 0.5 8>, angularLimitsMin=<-1 -55 -30>, angularTarget=<0 0 0>}
      localJointOffset=<-15 0 0>
     }]
    sphericalLimits [
     {limitRadius=5, drivingBone {boneName=neck_02}, sphereLocalOffset=<0 0 0>}
     {limitRadius=7, drivingBone {boneName=neck_01}, sphereLocalOffset=<5 10 0>}
    ]
   }
   n35 = Anim Dynamics - Bone: dyn_hair_fl_02(ComponentPose=@n34, Alpha=@n17, ExternalForce="0,0,0") {
    angularDampingOverride=0.99
    angularSpringConstant=1000
    bAngularSpring=true
    bGravityOverrideInSimSpace=true
    bOverrideAngularDamping=true
    gravityScale=0.25
    alphaBoolBlend {blendOption=Linear}
    boundBone {boneName=dyn_hair_fl_02}
    componentAppliedLinearAccClamp=<100 100 100>
    componentLinearVelScale=<50 50 50>
    physicsBodyDefinitions [{
      collisionType=CustomSphere
      sphereCollisionRadius=1
      boundBone {boneName=dyn_hair_fl_02}
      boxExtents=<5 30 30>
      constraintSetup {angularLimitsMax=<1 1 12>, angularLimitsMin=<-1 -55 -40>, angularTarget=<0 0 0>}
      localJointOffset=<-11 0 0>
     }]
    planarLimits [{drivingBone {boneName=spine_05}, planeTransform {location=<0 0 0>, rotation=<rot -90 90 0>}}]
    sphericalLimits [{limitRadius=30, drivingBone {boneName=spine_05}, sphereLocalOffset=<10 7 -24>}]
   }
   n36 = Anim Dynamics - Bone: dyn_hair_fl_03(ComponentPose=@n35, Alpha=@n17, ExternalForce="0,0,0") {
    angularDampingOverride=0.99
    angularSpringConstant=1000
    bAngularSpring=true
    bGravityOverrideInSimSpace=true
    bOverrideAngularDamping=true
    gravityScale=0.3
    alphaBoolBlend {blendOption=Linear}
    boundBone {boneName=dyn_hair_fl_03}
    componentAppliedLinearAccClamp=<100 100 100>
    componentLinearVelScale=<50 50 50>
    physicsBodyDefinitions [{
      collisionType=CustomSphere
      sphereCollisionRadius=1
      boundBone {boneName=dyn_hair_fl_03}
      boxExtents=<5 30 30>
      constraintSetup {angularLimitsMax=<3 5 15>, angularLimitsMin=<-3 -55 -45>, angularTarget=<0 0 0>}
      localJointOffset=<-4 0 0>
     }]
    planarLimits [{drivingBone {boneName=spine_05}, planeTransform {location=<0 0 0>, rotation=<rot -90 90 0>}}]
    sphericalLimits [{limitRadius=30, drivingBone {boneName=spine_05}, sphereLocalOffset=<10 7 -24>}]
   }
   n37 = "Transform (Modify) Bone - Bone: dyn_hair_l_02"(ComponentPose=@n36, Alpha=1, Translation="0,0,0", Rotation=Get HairLeftRotation() -> (HairLeftRotation:Rotator), Scale="1,1,1") {rotationMode=BMM_Additive, rotationSpace=BCS_BoneSpace, alphaBoolBlend {blendOption=Linear}, boneToModify {boneName=dyn_hair_l_02}}
   n38 = "Transform (Modify) Bone - Bone: dyn_hair_bk_r_01"(ComponentPose=@n37, Alpha=@n1, Translation="0,0,0", Rotation="-14,-10.000,-0.000", Scale="1,1,1") {
    rotationMode=BMM_Additive
    rotationSpace=BCS_BoneSpace
    alphaBoolBlend {blendOption=Linear}
    boneToModify {boneName=dyn_hair_bk_r_01}
    rotation {pitch=-14, yaw=-10}
   }
   n39 = "Transform (Modify) Bone - Bone: dyn_hair_bk_r_02"(ComponentPose=@n38, Alpha=@n1, Translation="0,0,0", Rotation="-15,6,-3", Scale="1,1,1") {
    rotationMode=BMM_Additive
    rotationSpace=BCS_BoneSpace
    alphaBoolBlend {blendOption=Linear}
    boneToModify {boneName=dyn_hair_bk_r_02}
    rotation=<rot -15 6 -3>
   }
   Save cached pose 'PreCrouchCache'(Pose=Component To Local(ComponentPose=@n39)) {cacheName=PreCrouchCache}
  ]}
 EventGraph {links [], nodes []}
]
variables [HairAlpha {type=number, default=0.875}]
